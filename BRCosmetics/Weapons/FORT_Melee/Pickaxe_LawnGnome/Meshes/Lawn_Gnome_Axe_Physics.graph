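# Physics asset for the Lawn Gnome pickaxe mesh, covering its rope bones.
# Three small capsule bodies sit on a master rope bone and two rope bones.
# Each of the two rope bodies is constrained directly to the master, forming a two-branch setup rather than a chain, with matching swing and twist angle limits on both joints.
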PhysicsAsset /BRCosmetics/Weapons/FORT_Melee/Pickaxe_LawnGnome/Meshes/Lawn_Gnome_Axe_Physics
bodies {
 dyn_master_rope [{shapeType=Capsule, length=1, radius=1}]
 dyn_rope1 [{shapeType=Capsule, length=2, radius=1.1, center=<1 0 0>, rotation=<rot 90 0 0>}]
 dyn_rope2 [{shapeType=Capsule, length=2, radius=1.1, center=<1 0 0>, rotation=<rot 90 0 0>}]
}
constraints [
 {bone1Name=dyn_rope1, bone2Name=dyn_master_rope, swing1LimitDegrees=45, swing2LimitDegrees=45, twistLimitDegrees=45}
 {bone1Name=dyn_rope2, bone2Name=dyn_master_rope, swing1LimitDegrees=45, swing2LimitDegrees=45, twistLimitDegrees=45}
]
